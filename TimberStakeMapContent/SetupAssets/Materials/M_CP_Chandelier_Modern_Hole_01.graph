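# Masked, two-sided Default Lit material for a modern chandelier prop, by its name the "Hole" variant, with Creative move tool support.
# The diffuse texture drives base color, opacity and opacity mask from its RGB output, while the S texture's green channel alone feeds metallic, specular and roughness.
# These and a normal map are packed into material attributes and passed through MF_CreativeMoveTool.
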Material /TimberStakeMapContent/SetupAssets/Materials/M_CP_Chandelier_Modern_Hole_01
graph {
 properties {bUseMaterialAttributes=true, blendMode=BLEND_Masked, pixelDepthOffsetMode=PDOM_Legacy, twoSided=true, usedShadingModels=Default Lit}
 nodes [
  n1 = TextureSample() {samplerType=SAMPLERTYPE_LinearColor, texture=/TimberStakeMapContent/SetupAssets/Textures/T_Chandelier_Modern_Hole_01_D}
  n2 = TextureSample() {samplerType=SAMPLERTYPE_LinearColor, texture=/TimberStakeMapContent/SetupAssets/Textures/T_Chandelier_Modern_Hole_01_S}
  MaterialFunctionCall(MA In=MakeMaterialAttributes!(
    BaseColor=@n1.RGB
    Metallic=@n2.G
    Specular=@n2.G
    Roughness=@n2.G
    Anisotropy
    EmissiveColor
    Opacity=@n1.RGB
    OpacityMask=@n1.RGB
    Normal=TextureSample() {samplerType=SAMPLERTYPE_Normal, texture=/TimberStakeMapContent/SetupAssets/Textures/T_Chandelier_Modern_Hole_01_N}.RGB
    Tangent
    WorldPositionOffset
    SubsurfaceColor
    ClearCoat
    ClearCoatRoughness
    AmbientOcclusion
    Refraction
    CustomizedUVs_0
    CustomizedUVs_1
    CustomizedUVs_2
    CustomizedUVs_3
    CustomizedUVs_4
    CustomizedUVs_5
    CustomizedUVs_6
    CustomizedUVs_7
    PixelDepthOffset
    ShadingModel
    Displacement
   )) {materialFunction=&MF_CreativeMoveTool}
 ]
}
